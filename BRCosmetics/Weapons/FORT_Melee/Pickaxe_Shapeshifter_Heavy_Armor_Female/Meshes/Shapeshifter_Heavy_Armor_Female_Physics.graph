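# Physics asset for the Shapeshifter Heavy Armor Female pickaxe mesh, setting up physics for the dynamic fur joints on the weapon.
# Small capsule bodies cover the two-handed handle bone and top and bottom fur joints.
# Each fur group has a small joint and a two-link large chain, both constrained to the handle.
# Narrow swing limits keep motion tight, and the large chains also lock twist.
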PhysicsAsset /BRCosmetics/Weapons/FORT_Melee/Pickaxe_Shapeshifter_Heavy_Armor_Female/Meshes/Shapeshifter_Heavy_Armor_Female_Physics
bodies {
 Melee_TwoHanded_Handle [{shapeType=Capsule, length=2, radius=2.5, rotation=<rot 90 0 0>}]
 dyn_bottomFur_large_jnt_1 [{shapeType=Capsule, length=2, radius=2.5, center=<3.5535 0 0>, rotation=<rot 90 0 0>}]
 dyn_bottomFur_large_jnt_2 [{shapeType=Capsule, length=2, radius=2.5, center=<3.5535 0 0>, rotation=<rot 90 0 0>}]
 dyn_bottomFur_small_jnt [{shapeType=Capsule, length=2, radius=2.5, center=<2.8471 0.66 -0.0001>, rotation=<rot 90 0 -15.7677>}]
 dyn_topFur_large_jnt_1 [{shapeType=Capsule, length=2, radius=2.5, center=<3.5535 0 0>, rotation=<rot 90 0 0>}]
 dyn_topFur_large_jnt_2 [{shapeType=Capsule, length=2, radius=2.5, center=<3.5535 0 0>, rotation=<rot 90 0 0>}]
 dyn_topFur_small_jnt [{shapeType=Capsule, length=2, radius=2.5, center=<2.8471 0.66 -0.0001>, rotation=<rot 90 0 -15.7677>}]
}
constraints [
 {bone1Name=dyn_topFur_small_jnt, bone2Name=Melee_TwoHanded_Handle, swing1LimitDegrees=10, swing2LimitDegrees=10, twistLimitDegrees=10}
 {
  bone1Name=dyn_topFur_large_jnt_1
  bone2Name=Melee_TwoHanded_Handle
  swing1LimitDegrees=20
  swing2LimitDegrees=10
  twistLimitDegrees=3
  twistMotion=Locked
 }
 {
  bone1Name=dyn_topFur_large_jnt_2
  bone2Name=dyn_topFur_large_jnt_1
  swing1LimitDegrees=20
  swing2LimitDegrees=8
  twistLimitDegrees=1
  twistMotion=Locked
 }
 {bone1Name=dyn_bottomFur_small_jnt, bone2Name=Melee_TwoHanded_Handle, swing1LimitDegrees=10, swing2LimitDegrees=10, twistLimitDegrees=10}
 {
  bone1Name=dyn_bottomFur_large_jnt_1
  bone2Name=Melee_TwoHanded_Handle
  swing1LimitDegrees=20
  swing2LimitDegrees=5
  twistLimitDegrees=10
  twistMotion=Locked
 }
 {
  bone1Name=dyn_bottomFur_large_jnt_2
  bone2Name=dyn_bottomFur_large_jnt_1
  swing1LimitDegrees=10
  swing2LimitDegrees=5
  twistLimitDegrees=4
  twistMotion=Locked
 }
]
